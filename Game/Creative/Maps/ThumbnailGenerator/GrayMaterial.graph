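# GrayMaterial is a Default Lit material in the Creative ThumbnailGenerator map folder, apparently a plain gray surface for thumbnail generation.
# It exposes a Base Color vector parameter defaulting to dark gray and a Metallic scalar parameter, alongside two fixed scalar constants.
# It contains no textures or math nodes, and no connections show which material inputs the constants drive.
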Material /Game/Creative/Maps/ThumbnailGenerator/GrayMaterial
graph {
 properties {pixelDepthOffsetMode=PDOM_Legacy, usedShadingModels=Default Lit}
 nodes [
  Constant() {r=0.2}
  Constant() {r=1}
  ScalarParameter() {parameterName=Metallic}
  VectorParameter() {parameterName=Base Color, defaultValue=<rgba 0.08 0.08 0.08 0>}
 ]
}
